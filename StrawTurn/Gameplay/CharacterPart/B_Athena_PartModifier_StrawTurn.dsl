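; Character part modifier for the StrawTurn cosmetic that suppresses the player's own particle effects on specific characters.
; On post-res-in it binds the idle Niagara effect to the part's skeletal mesh.
; If the loadout character is in ToCheckCID, it caches and hides Body, Face and Head FX components.
; On destruction it unhides those cached components; a debug event prints the idle effect's attachment, and other overrides only call the parent.
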(event EventDestroyed
  ; Turns on Effects if Head CP is removed without Character Death
  (|CallParentFunction)
  (CallFunction|DEBUGPrintString "DETROYED")
  (CallFunction|CIDChecker true))

(event Custom|AttachmentTest
  ;[1 1 1] Added Print String Test for QA to see what's attached where
  (Development|PrintString (Utilities|String|Append "Idle VFX Attached to - " (Utilities|GetDisplayName (Transformation|GetAttachParent (CallFunction|GetIdleVFXComponent))) " - at Socket - " (Utilities|String|ToString(Name) (Transformation|GetAttachSocketName (CallFunction|GetIdleVFXComponent)))) true true "(R=1.000000,G=0.023320,B=0.000000,A=1.000000)" 1e+01))

(event Custom|CID%20Checker
  (params (bisHeadRemoved :type Boolean))
  ; Removes All Niagara or Cascade Particles Systems on CIDs which are called out in ToCheckCID
  (if bisHeadRemoved
    (for _array_element (Variables|Default|GetModified_Body)
      (Utilities|IsValid _array_element
        (:Is%20Valid
          (Development|SetHiddeninGame _array_element false true))))
    (for _array_element_2 (Variables|Default|GetModified_Face)
      (Utilities|IsValid _array_element_2
        (:Is%20Valid
          (Development|SetHiddeninGame _array_element_2 false true))))
    (for _array_element_3 (Variables|Default|GetModified%20Head)
      (Utilities|IsValid _array_element_3
        (:Is%20Valid
          (Development|SetHiddeninGame _array_element_3 false true))))
    (else
      (if (>= (Utilities|Array|FindItem (Variables|Default|GetToCheckCID) (Utilities|ResolveSoftReference (CosmeticLoadout|GetItemDefinitionTypefromLoadout (CosmeticLoadout|GetCosmeticLoadout) "(TagName=\"LoadoutSlot.Character\")" "/Script/FortniteGame.AthenaCharacterItemDefinition"))) 0)
        (bind _output_get (Variables|Default|SetModified_Body (Customization|GetComponentsbyClassforPartType (CustomCharacter|GetAssociatedPlayerPawn) "/Script/Engine.FXSystemComponent" "Body")))
        (for _array_element_4 _output_get
          (Development|SetHiddeninGame _array_element_4 true true))
        (bind _output_get_2 (Variables|Default|SetModified_Face (Customization|GetComponentsbyClassforPartType (CustomCharacter|GetAssociatedPlayerPawn) "/Script/Engine.FXSystemComponent" "Face")))
        (for _array_element_5 _output_get_2
          (Development|SetHiddeninGame _array_element_5 true true))
        (bind _output_get_3 (Variables|Default|SetModified%20Head (Customization|GetComponentsbyClassforPartType (CustomCharacter|GetAssociatedPlayerPawn) "/Script/Engine.FXSystemComponent" "Head")))
        (for _array_element_6 _output_get_3
          (Development|SetHiddeninGame _array_element_6 true true))))))

(event EventOnPostResIn
  (|CallParentFunction)
  (Niagara|SetNiagaraSkeletalMeshComponent (Utilities|Casting|CastToNiagaraComponent (CallFunction|GetIdleVFXComponent)) "SkeletalMesh" (Customization|GetSkeletalMeshforPartType (CustomCharacter|GetAssociatedPlayerPawn) (CustomCharacter|GetCharacterPartType) :!class /Script/FortniteGame.FortPlayerPawn))
  (CallFunction|CIDChecker false))

(event EventOnTargetingOrFirstPersonCameraChanged
  (|CallParentFunction))

(event EventOnEnterVehicle
  (params (VehicleEntered :type Fort%20Vehicle%20Interface%20Interface%20%28by%20ref%29))
  (|CallParentFunction VehicleEntered))

(event EventOnExitVehicle
  (|CallParentFunction))

(event EventOnBeginSkydiving
  (|CallParentFunction))

(event EventOnEndSkydiving
  (|CallParentFunction))

(fn ConstructionScript
  (|CallParentFunction))
